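# Conversation database for the QuickBurst NPC, defining the greeting and top-level branches when a player talks to this character.
# The entry point feeds one Speech node, whose contextual lines cover first-ever meetings, per-match conversation counts and a specific-outfit reaction, with priority curves on the first and last.
# Six link nodes then branch by remote entry tag to shared quests, basic services, QuickBurst dialogue and Season 30 and 31 story conversations.
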ConversationDatabase /NPCLibrary/NPCs/QuickBurst/NPCLibrary_Conversation_QuickBurst
entry_points [{tag=Conversation.Entry.QuickBurst, destinations [#1]}]
nodes [
 ConversationEntryPointNode_7 {
  id=#1
  class=ConversationEntryPointNode
  outputs [#2]
  properties {entryTag {tagName=Conversation.Entry.QuickBurst}}
 }
 Speech_C_0 {
  id=#2
  class=Speech_C
  outputs [
   #3
   #4
   #5
   #6
   #7
   #8
  ]
  properties {
   nodeName=""
   generalConfig {
    defaultMessage="I miss the thrashin', scrappin' sands..."
    contextualMessages [
     {
      message="The wasteland is gone... in spirit, at least..."
      contextRequirements [{requirement {class=FortControllerRequirement_IsFirstEverConversationWithNPC}}]
      priority {curve {rowName=Tandem.Speech.Priority.FirstEverInteraction}}
     }
     {
      message="I miss the thrashin', scrappin' sands..."
      contextRequirements [{requirement {class=FortControllerRequirement_CheckNPCMatchConversationCount, properties {nPCUniqueIdTag {tagName=AISpawnerData.Type.Tandem.QuickBurst}}}}]
      priority {curve {rowName=None}}
     }
     {
      message=This place is so different now.
      contextRequirements [{requirement {class=FortControllerRequirement_CheckNPCMatchConversationCount, properties {nPCUniqueIdTag {tagName=AISpawnerData.Type.Tandem.QuickBurst}}}}]
      priority {curve {rowName=None}}
     }
     {
      message=*sigh* I got sand in my teeth again. Just like old times.
      contextRequirements [{requirement {class=FortControllerRequirement_CheckNPCMatchConversationCount, properties {nPCUniqueIdTag {tagName=AISpawnerData.Type.Tandem.QuickBurst}}}}]
      priority {curve {rowName=None}}
     }
     {
      message=Doom took down the wasteland. So we gotta take down Doom.
      contextRequirements [{requirement {
         class=FortControllerRequirement_HasCID
         properties {allowedCIds [/BRCosmetics/Athena/Items/Cosmetics/Characters/CID_795_Athena_Commando_M_Dummeez]}
        }}]
      priority {curve {rowName=Tandem.Speech.Priority.OutfitReaction}}
     }
    ]
   }
  }
 }
 ConversationLinkNode_5 {
  id=#8
  class=ConversationLinkNode
  properties {remoteEntryTag {tagName=Conversation.Entry.Season31.Story.P04.Q03}}
 }
 ConversationLinkNode_4 {
  id=#7
  class=ConversationLinkNode
  properties {remoteEntryTag {tagName=Conversation.Entry.Season30.Story.P06.Q01.QB}}
 }
 ConversationLinkNode_3 {
  id=#6
  class=ConversationLinkNode
  properties {remoteEntryTag {tagName=Conversation.Entry.QuickBurst.Dialogue}}
 }
 ConversationLinkNode_2 {
  id=#5
  class=ConversationLinkNode
  properties {remoteEntryTag {tagName=Conversation.Entry.Season30.Story.P02.01.QuickBurst}}
 }
 ConversationLinkNode_0 {id=#4, class=ConversationLinkNode, properties {remoteEntryTag {tagName=Conversation.Entry.BasicServices}}}
 ConversationLinkNode_1 {id=#3, class=ConversationLinkNode, properties {remoteEntryTag {tagName=Conversation.Entry.NPCLibrary.Quests}}}
]
