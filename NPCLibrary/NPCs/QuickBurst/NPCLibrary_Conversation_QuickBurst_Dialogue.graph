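# Conversation database for the QuickBurst NPC's ask-about dialogue, entered through the Conversation.Entry.QuickBurst.Dialogue tag.
# An opening speech offers Back, two topic choices (The Wasteland and Doom) and a link to a separate Season 31 Found GreenTown QuickBurst entry.
# Each topic plays a three-line speech chain, and both end at a shared Return To Start node labeled BACK.
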ConversationDatabase /NPCLibrary/NPCs/QuickBurst/NPCLibrary_Conversation_QuickBurst_Dialogue
entry_points [{tag=Conversation.Entry.QuickBurst.Dialogue, destinations [#1]}]
nodes [
 ConversationEntryPointNode_0 {
  id=#1
  class=ConversationEntryPointNode
  outputs [#2]
  properties {entryTag {tagName=Conversation.Entry.QuickBurst.Dialogue}}
 }
 Speech_C_0 {
  id=#2
  class=Speech_C
  outputs [#3, #4, #5, #6]
  properties {
   nodeName=""
   softTaskIcon=&T_UI_NPC_question
   generalConfig {defaultMessage="Yeah, I can spare a clang. What's up?"}
  }
  sub_nodes [ConversationChoiceNode_5 {class=ConversationChoiceNode, properties {defaultChoiceDisplayText=Ask about...}}]
 }
 ConversationLinkNode_0 {
  id=#6
  class=ConversationLinkNode
  properties {remoteEntryTag {tagName=Conversation.Entry.Season31.Found.GreenTown.QuickBurst}}
 }
 Speech_C_5 {
  id=#5
  class=Speech_C
  outputs [#7]
  properties {nodeName="", softTaskIcon=&T_UI_NPC_question, generalConfig {defaultMessage="Sure, he got rid of Megalo Don..."}}
  sub_nodes [ConversationChoiceNode_1 {class=ConversationChoiceNode, properties {defaultChoiceDisplayText=Doom}}]
 }
 Speech_C_6 {
  id=#7
  class=Speech_C
  outputs [#8]
  properties {
   nodeName=""
   softTaskIcon=&T_UI_NPC_question
   generalConfig {defaultMessage=...but he crushed the spirit of the Wasteland along the way.}
  }
 }
 Speech_C_8 {
  id=#8
  class=Speech_C
  outputs [#9]
  properties {
   nodeName=""
   softTaskIcon=&T_UI_NPC_question
   generalConfig {defaultMessage=He'll pay for that. I'll do whatever it takes to help Hope take him down.}
  }
 }
 ReturnToStart_C_23 {
  id=#9
  class=ReturnToStart_C
  properties {softTaskIcon=&T_UI_NPC_Back}
  sub_nodes [ConversationChoiceNode_32 {class=ConversationChoiceNode, properties {defaultChoiceDisplayText=BACK}}]
 }
 Speech_C_2 {
  id=#4
  class=Speech_C
  outputs [#10]
  properties {
   nodeName=""
   softTaskIcon=&T_UI_NPC_question
   generalConfig {defaultMessage=The spirit of the Wasteland blew up with the pipeline.}
  }
  sub_nodes [ConversationChoiceNode_0 {class=ConversationChoiceNode, properties {defaultChoiceDisplayText=The Wasteland}}]
 }
 Speech_C_3 {
  id=#10
  class=Speech_C
  outputs [#11]
  properties {
   nodeName=""
   softTaskIcon=&T_UI_NPC_question
   generalConfig {defaultMessage="Sure the sand is still there, but... it's just not the same."}
  }
 }
 Speech_C_4 {
  id=#11
  class=Speech_C
  outputs [#9]
  properties {
   nodeName=""
   softTaskIcon=&T_UI_NPC_question
   generalConfig {defaultMessage=I'll find somewhere else that gives me that rowdy feeling... I just know it.}
  }
 }
 Back_C_0 {
  id=#3
  class=Back_C
  sub_nodes [ConversationChoiceNode_7 {class=ConversationChoiceNode, properties {defaultChoiceDisplayText=BACK}}]
 }
]
